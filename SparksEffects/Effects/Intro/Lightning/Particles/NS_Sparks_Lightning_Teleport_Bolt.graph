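# Niagara lightning bolt burst for the Sparks intro teleport effect, built from short-lived meshes aligned to a point high above the system origin.
# Cyan Primary and center emitters fire immediately, while thinner magenta bolts follow slightly later.
# Each spawns on a small cylinder with randomized scale and rotation, then flickers through color, alpha and dynamic material parameter curves.
# The center emitter also drives a flickering violet-blue light renderer offset above each particle.
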NiagaraSystem /SparksEffects/Effects/Intro/Lightning/Particles/NS_Sparks_Lightning_Teleport_Bolt
dependencies {usedDataInterfaces [&NiagaraDataInterfaceCurve], usedDynamicInputs [], usedModules [&SystemState]}
emitters {
 Lightning_Primary {
  emitterUpdateScript [
   &EmitterState()
   &SpawnBurst_Instantaneous(Use Spawn Probability=false, UseLoopCountLimit=false)
   &SpawnBurst_Instantaneous(Spawn Count=4, Spawn Time=0.07, Use Spawn Probability=false, UseLoopCountLimit=false) {enabled=false, module=SpawnBurst_Instantaneous001}
  ]
  particleSpawnScript [
   /Niagara/Modules/Spawn/Initialization/InitializeParticle(
    Color=<rgba 0 0.9504 3 1>
    Lifetime=0.3
    Mesh Scale=&RandomRangeVector(Minimum=<0.5 0.5 0.8>, Maximum=<1 1 1.5>)
    Module.Write Color=true
    Module.Write Lifetime=true
    Module.Write Mass=false
    Module.Write Material Random=true
    Module.Write Position=true
    Module.Write Scale=true
    Module.Write SpriteRotation=false
    Module.Write SpriteSize=false
    Position=&SimulationPosition()
   ) {enabled=false}
   /Niagara/Modules/Spawn/Initialization/V2/InitializeParticle(
    Color=<rgba 0 0.9504 3 1>
    Lifetime=0.3
    Mesh Scale=&RandomRangeVector(Minimum=<0.5 0.5 0.8>, Maximum=<1 1 1.5>)
    Mesh Scale Mode=::NewEnumerator4
    Position=&SimulationPosition()
    Position Mode=::NewEnumerator1
    Sprite Size Mode=::NewEnumerator4
    UsePositionOffset=false
    Write Material Random=true
   ) {module=InitializeParticle001}
   &InitialMeshOrientation(
    Orientation Axis=<-1 0 0>
    Rotation=&RandomRangeVector(Minimum=<0 0 0>, Maximum=<1 0 0>)
    Use Orientation Vector=true
    Use Rotation Vector=true
   ) {module=InitialMeshOrientation001}
   &DynamicMaterialParameters(
    Index 0 Param 2=0.1
    Index 0 Param 4=1
    Module.Param0WriteEnabled=true
    Module.Param1WriteEnabled=true
    Module.Param2WriteEnabled=true
    Module.Param3WriteEnabled=false
   )
   &CylinderLocation(Cylinder Height=25, Cylinder Radius=1, MaskSpawnGroup=false, Override Seed=false, Surface Only=true, UseOwnerScale=true)
   &CylinderLocation(Cylinder Radius=12000, MaskSpawnGroup=false, Override Seed=false, Surface Only=true, UseOwnerScale=true) {enabled=false, module=CylinderLocation001}
   /Niagara/Modules/Update/Orientation/V2/OrientMeshToVector(
    Look At Direction:Vector3f=&Subtract_Vector(
     A=Linked() {linkedVariable {name=Particles.Position, type=NiagaraPosition}}
     B=&Add_Vector(A=/Niagara/DynamicInputs/Helpers/SimulationPosition(), B=&MakeVector(X=0, Y=0, Z=30000))
    )
    Override Quat=false
    UseOptionalSideAxis=false
    UseOptionalUpAxis=false
   )
   /SetVariables(Particles.MeshIndex=1) {isSetParameters=true, module=SetVariables_#1}
  ]
  particleUpdateScript [
   &ParticleState()
   &ScaleMeshSize(Scale Factor=<0.5 0.5 0.5>)
   &ScaleColor(
    Scale Alpha=&FloatFromCurve(FloatCurve=DataInterface() {propertyValues {Curve {keys [
         <at 0 0>
         {interpMode=RCIM_Cubic, tangentMode=RCTM_User, time=0.1258, value=0.1952}
         <at 0.341 0>
         {interpMode=RCIM_Cubic, tangentMode=RCTM_User, time=0.496, value=0.2532}
         <at 1 0>
        ]}}})
    Scale RGB=&VectorFromFloat(Value=&FloatFromCurve(FloatCurve=DataInterface() {propertyValues {Curve {keys [
          <at 0.4 0.5 RCIM_Cubic>
          {arriveTangent=-2.404, interpMode=RCIM_Cubic, leaveTangent=-2.404, tangentMode=RCTM_User, time=0.45, value=1}
          <at 1 0.5>
         ]}}}))
    ScaleA=true
    ScaleRGB=true
   )
   &DynamicMaterialParameters(
    Index 0 Param 1=&FloatFromCurve(FloatCurve=DataInterface() {propertyValues {Curve {keys [
         {arriveTangent=0.2699, interpMode=RCIM_Cubic, leaveTangent=0.2699, tangentMode=RCTM_User, time=0, value=1}
         {arriveTangent=-2.2027, interpMode=RCIM_Cubic, leaveTangent=-2.2027, tangentMode=RCTM_User, time=1, value=0}
        ]}}})
    Index 0 Param 2=&FloatFromCurve(FloatCurve=DataInterface() {propertyValues {Curve {keys [
         {interpMode=RCIM_Cubic, leaveTangent=-1.1405, tangentMode=RCTM_Break, time=0, value=0.5}
         {arriveTangent=4.4936, interpMode=RCIM_Cubic, leaveTangent=-5.3647, tangentMode=RCTM_Break, time=0.4833, value=1}
         <at 1 0>
        ]}}})
    Index 0 Param 3=0.015
    Module.Param0WriteEnabled=true
    Module.Param1WriteEnabled=false
    Module.Param2WriteEnabled=true
    Module.Param3WriteEnabled=false
   ) {module=DynamicMaterialParameters001}
   /SetVariables(Particles.light color:LinearColor=&MakeLinearColorFromVectorAndFloat(
     "Vector (RGB)"=&Multiply_VectorByFloat(
      Vector=&MakeVectorFromLinearColorRGB(Color=<rgba 0.6417 0.4419 3 1>)
      Float=&FloatFromCurve(
       FloatCurve=DataInterface() {propertyValues {Curve {keys [
           {interpMode=RCIM_Cubic, tangentMode=RCTM_User, time=0, value=1}
           {arriveTangent=-0.1663, interpMode=RCIM_Cubic, leaveTangent=-0.1663, tangentMode=RCTM_User, time=0.0918, value=0.1}
           {arriveTangent=-0.2815, interpMode=RCIM_Cubic, leaveTangent=-0.2815, tangentMode=RCTM_User, time=0.1514, value=0.9748}
           {arriveTangent=-0.2665, interpMode=RCIM_Cubic, leaveTangent=-0.2665, tangentMode=RCTM_User, time=0.2547, value=0}
           {arriveTangent=0.512, interpMode=RCIM_Cubic, leaveTangent=0.512, tangentMode=RCTM_User, time=0.5641, value=0.8648}
           {arriveTangent=-1.9839, interpMode=RCIM_Cubic, leaveTangent=-1.9839, tangentMode=RCTM_User, time=0.8, value=0.05}
           {arriveTangent=-0.25, interpMode=RCIM_Cubic, leaveTangent=-0.25, time=0.9, value=0.45}
           <at 1 0>
          ]}}}
       Scale Curve=25
      )
     )
     "Float (Alpha)"=1
    )) {enabled=false, isSetParameters=true, module=SetVariables_#2}
   /SetVariables(Particles.LightPosition:Vector3f=&Add_Vector(A=Linked() {linkedVariable {name=Particles.Position, type=NiagaraPosition}}, B=<0 0 75>)) {enabled=false, isSetParameters=true, module=SetVariables_#3}
  ]
  rendererClasses [&NiagaraMeshRendererProperties]
  renderers [{rendererClass=&NiagaraMeshRendererProperties, rendererIndex=0}]
 }
 Lightning_Thin {
  emitterUpdateScript [
   &EmitterState()
   &SpawnBurst_Instantaneous(Spawn Count=3, Spawn Time=0.08, Use Spawn Probability=false, UseLoopCountLimit=false)
   &SpawnBurst_Instantaneous(Spawn Time=0.19, Use Spawn Probability=false, UseLoopCountLimit=false) {enabled=false, module=SpawnBurst_Instantaneous001}
  ]
  particleSpawnScript [
   /Niagara/Modules/Spawn/Initialization/InitializeParticle(
    Color=<rgba 1.4708 0.0432 2.5 1>
    Lifetime=0.15
    Mesh Scale=&RandomRangeVector(Minimum=<10 10 50>, Maximum=<60 60 60>)
    Module.Write Color=true
    Module.Write Lifetime=true
    Module.Write Mass=false
    Module.Write Material Random=true
    Module.Write Position=true
    Module.Write Scale=true
    Module.Write SpriteRotation=false
    Module.Write SpriteSize=false
    Position=/Niagara/Functions/Localspace/SimulationPosition()
   ) {enabled=false}
   /Niagara/Modules/Spawn/Initialization/V2/InitializeParticle(
    Color=<rgba 1.4708 0.0432 2.5 1>
    Lifetime=0.15
    Mesh Scale=&RandomRangeVector(Minimum=<10 10 50>, Maximum=<60 60 60>)
    Mesh Scale Mode=::NewEnumerator4
    Position=&SimulationPosition()
    Position Mode=::NewEnumerator1
    Sprite Size Mode=::NewEnumerator4
    UsePositionOffset=false
    Write Material Random=true
   ) {module=InitializeParticle001}
   &InitialMeshOrientation(
    Orientation Axis=<-1 0 0>
    Rotation=&RandomRangeVector(Minimum=<0 0 0>, Maximum=<1 0 0>)
    Use Orientation Vector=true
    Use Rotation Vector=true
   ) {module=InitialMeshOrientation001}
   &DynamicMaterialParameters(
    Index 0 Param 2=&UniformRangedFloat(Minimum=1, Maximum=1)
    Index 0 Param 3=1
    Index 0 Param 4=1
    Module.Param0WriteEnabled=false
    Module.Param1WriteEnabled=true
    Module.Param2WriteEnabled=true
    Module.Param3WriteEnabled=false
   )
   &CylinderLocation(Cylinder Height=25, Cylinder Radius=1, MaskSpawnGroup=false, Override Seed=false, Surface Only=true, UseOwnerScale=true)
   &OrientMeshToVector(
    Look At Direction:Vector3f=&Subtract_Vector(
     A=Linked() {linkedVariable {name=Particles.Position, type=NiagaraPosition}}
     B=&Add_Vector(A=/Niagara/DynamicInputs/Helpers/SimulationPosition(), B=&MakeVector(X=0, Y=0, Z=30000))
    )
    Override Quat=false
    UseOptionalSideAxis=false
    UseOptionalUpAxis=false
   )
   /SetVariables(Particles.MeshIndex=&RandomRangeInt(Minimum=0, Maximum=1)) {isSetParameters=true, module=SetVariables_#1}
  ]
  particleUpdateScript [
   &ParticleState()
   &ScaleMeshSize(Scale Factor=&VectorFromFloat(Value=0.008))
   &ScaleColor(
    Scale Alpha=&FloatFromCurve(FloatCurve=DataInterface() {propertyValues {Curve {keys [
         {arriveTangent=3.6697, interpMode=RCIM_Cubic, leaveTangent=3.6697, tangentMode=RCTM_User, time=0, value=0}
         {interpMode=RCIM_Cubic, tangentMode=RCTM_User, time=0.1258, value=0.1952}
         <at 0.2 0>
         {arriveTangent=5.3955, interpMode=RCIM_Cubic, leaveTangent=5.3955, tangentMode=RCTM_User, time=0.4, value=0}
         {arriveTangent=-1.3583, interpMode=RCIM_Cubic, leaveTangent=-1.3583, tangentMode=RCTM_User, time=0.496, value=0.2532}
         <at 1 0>
        ]}}})
    Scale RGB=&VectorFromFloat(Value=&FloatFromCurve(FloatCurve=DataInterface() {propertyValues {Curve {keys [
          <at 0.4 0.5 RCIM_Cubic>
          {arriveTangent=-2.404, interpMode=RCIM_Cubic, leaveTangent=-2.404, tangentMode=RCTM_User, time=0.45, value=1}
          <at 1 0.5>
         ]}}}))
    ScaleA=true
    ScaleRGB=true
   )
   &DynamicMaterialParameters(
    Index 0 Param 1=&FloatFromCurve(FloatCurve=DataInterface() {propertyValues {Curve {keys [<at 0 1>,<at 1 0>]}}})
    Index 0 Param 2=&FloatFromCurve(
     FloatCurve=DataInterface() {propertyValues {Curve {keys [
         {interpMode=RCIM_Cubic, leaveTangent=-1.1405, tangentMode=RCTM_Break, time=0, value=0.5}
         {arriveTangent=4.4936, interpMode=RCIM_Cubic, leaveTangent=-5.3647, tangentMode=RCTM_Break, time=0.4833, value=1}
         <at 1 0>
        ]}}}
     Scale Curve=5
    )
    Index 0 Param 3=&RandomRangeFloat(Minimum=0, Maximum=0.005)
    Module.Param0WriteEnabled=true
    Module.Param1WriteEnabled=true
    Module.Param2WriteEnabled=true
    Module.Param3WriteEnabled=false
   ) {module=DynamicMaterialParameters001}
  ]
  rendererClasses [&NiagaraMeshRendererProperties]
  renderers [{rendererClass=&NiagaraMeshRendererProperties, rendererIndex=0}]
 }
 Lightning_center {
  emitterUpdateScript [
   &EmitterState()
   &SpawnBurst_Instantaneous(Spawn Count=2, Use Spawn Probability=false, UseLoopCountLimit=false)
   &SpawnBurst_Instantaneous(Spawn Time=0.07, Use Spawn Probability=false, UseLoopCountLimit=false) {enabled=false, module=SpawnBurst_Instantaneous001}
  ]
  particleSpawnScript [
   /Niagara/Modules/Spawn/Initialization/InitializeParticle(
    Color=<rgba 0 0.9504 3 1>
    Lifetime=0.3
    Mesh Scale=&RandomRangeVector(Minimum=<25 25 40>, Maximum=<50 50 60>)
    Module.Write Color=true
    Module.Write Lifetime=true
    Module.Write Mass=false
    Module.Write Material Random=true
    Module.Write Position=true
    Module.Write Scale=true
    Module.Write SpriteRotation=false
    Module.Write SpriteSize=false
    Position=/Niagara/Functions/Localspace/SimulationPosition()
   ) {enabled=false}
   /Niagara/Modules/Spawn/Initialization/V2/InitializeParticle(
    Color=<rgba 0 0.9504 3 1>
    Lifetime=0.3
    Mesh Scale=&RandomRangeVector(Minimum=<25 25 40>, Maximum=<50 50 60>)
    Mesh Scale Mode=::NewEnumerator4
    Position=&SimulationPosition()
    Position Mode=::NewEnumerator1
    Sprite Size Mode=::NewEnumerator4
    UsePositionOffset=false
    Write Material Random=true
   ) {module=InitializeParticle001}
   &InitialMeshOrientation(
    Orientation Axis=<-1 0 0>
    Rotation=&RandomRangeVector(Minimum=<0 0 0>, Maximum=<1 0 0>)
    Use Orientation Vector=true
    Use Rotation Vector=true
   ) {module=InitialMeshOrientation001}
   &DynamicMaterialParameters(
    Index 0 Param 2=&UniformRangedFloat(Minimum=0.1, Maximum=0.3)
    Index 0 Param 3=1
    Index 0 Param 4=1
    Module.Param0WriteEnabled=false
    Module.Param1WriteEnabled=true
    Module.Param2WriteEnabled=true
    Module.Param3WriteEnabled=false
   )
   &CylinderLocation(Cylinder Height=25, Cylinder Radius=1, MaskSpawnGroup=false, Override Seed=false, Surface Only=true, UseOwnerScale=true)
   &CylinderLocation(Cylinder Radius=12000, MaskSpawnGroup=false, Override Seed=false, Surface Only=true, UseOwnerScale=true) {enabled=false, module=CylinderLocation001}
   &OrientMeshToVector(
    Look At Direction:Vector3f=&Subtract_Vector(
     A=Linked() {linkedVariable {name=Particles.Position, type=NiagaraPosition}}
     B=&Add_Vector(A=/Niagara/DynamicInputs/Helpers/SimulationPosition(), B=&MakeVector(X=0, Y=0, Z=30000))
    )
    Override Quat=false
    UseOptionalSideAxis=false
    UseOptionalUpAxis=false
   )
   /SetVariables(Particles.MeshIndex=&RandomRangeInt(Minimum=0, Maximum=1)) {isSetParameters=true, module=SetVariables_#1}
  ]
  particleUpdateScript [
   &ParticleState()
   &ScaleMeshSize(Scale Factor=&VectorFromFloat(Value=0.008))
   &ScaleColor(
    Scale Alpha=&FloatFromCurve(FloatCurve=DataInterface() {propertyValues {Curve {keys [
         <at 0 0>
         {interpMode=RCIM_Cubic, tangentMode=RCTM_User, time=0.1258, value=0.1952}
         <at 0.341 0>
         {interpMode=RCIM_Cubic, tangentMode=RCTM_User, time=0.496, value=0.2532}
         <at 1 0>
        ]}}})
    Scale RGB=&VectorFromFloat(Value=&FloatFromCurve(FloatCurve=DataInterface() {propertyValues {Curve {keys [
          <at 0.4 0.5 RCIM_Cubic>
          {arriveTangent=-2.404, interpMode=RCIM_Cubic, leaveTangent=-2.404, tangentMode=RCTM_User, time=0.45, value=1}
          <at 1 0.5>
         ]}}}))
    ScaleA=true
    ScaleRGB=true
   )
   &DynamicMaterialParameters(
    Index 0 Param 1=&FloatFromCurve(FloatCurve=DataInterface() {propertyValues {Curve {keys [<at 0 1>,<at 1 0>]}}})
    Index 0 Param 2=&FloatFromCurve(
     FloatCurve=DataInterface() {propertyValues {Curve {keys [
         {interpMode=RCIM_Cubic, leaveTangent=-1.1405, tangentMode=RCTM_Break, time=0, value=0.5}
         {arriveTangent=4.4936, interpMode=RCIM_Cubic, leaveTangent=-5.3647, tangentMode=RCTM_Break, time=0.4833, value=1}
         <at 1 0>
        ]}}}
     Scale Curve=5
    )
    Index 0 Param 3=0.005
    Module.Param0WriteEnabled=true
    Module.Param1WriteEnabled=true
    Module.Param2WriteEnabled=true
    Module.Param3WriteEnabled=false
   ) {module=DynamicMaterialParameters001}
   /SetVariables(Particles.light color:LinearColor=&MakeLinearColorFromVectorAndFloat(
     "Vector (RGB)"=&Multiply_VectorByFloat(
      Vector=&MakeVectorFromLinearColorRGB(Color=<rgba 0.6417 0.4419 3 1>)
      Float=&FloatFromCurve(
       FloatCurve=DataInterface() {propertyValues {Curve {keys [
           {interpMode=RCIM_Cubic, tangentMode=RCTM_User, time=0, value=1}
           {arriveTangent=-0.1663, interpMode=RCIM_Cubic, leaveTangent=-0.1663, tangentMode=RCTM_User, time=0.0918, value=0.1}
           {arriveTangent=-0.2815, interpMode=RCIM_Cubic, leaveTangent=-0.2815, tangentMode=RCTM_User, time=0.1514, value=0.9748}
           {arriveTangent=-0.2665, interpMode=RCIM_Cubic, leaveTangent=-0.2665, tangentMode=RCTM_User, time=0.2547, value=0}
           {arriveTangent=0.512, interpMode=RCIM_Cubic, leaveTangent=0.512, tangentMode=RCTM_User, time=0.5641, value=0.8648}
           {arriveTangent=-1.9839, interpMode=RCIM_Cubic, leaveTangent=-1.9839, tangentMode=RCTM_User, time=0.8, value=0.05}
           {arriveTangent=-0.25, interpMode=RCIM_Cubic, leaveTangent=-0.25, time=0.9, value=0.45}
           <at 1 0>
          ]}}}
       Scale Curve=25
      )
     )
     "Float (Alpha)"=1
    )) {isSetParameters=true, module=SetVariables_#2}
   /SetVariables(Particles.LightPosition:Vector3f=&Add_Vector(A=Linked() {linkedVariable {name=Particles.Position, type=NiagaraPosition}}, B=<0 0 75>)) {isSetParameters=true, module=SetVariables_#3}
  ]
  rendererClasses [&NiagaraMeshRendererProperties, &NiagaraLightRendererProperties]
  renderers [
   {rendererClass=&NiagaraMeshRendererProperties, rendererIndex=0}
   {rendererClass=&NiagaraLightRendererProperties, rendererIndex=1}
  ]
 }
}
summary {
 systemName=NS_Sparks_Lightning_Teleport_Bolt
 emitters [
  {
   emitterName=Lightning_center
   rendererClasses [&NiagaraMeshRendererProperties, &NiagaraLightRendererProperties]
  }
  {emitterName=Lightning_Thin, rendererClasses [&NiagaraMeshRendererProperties]}
  {emitterName=Lightning_Primary, rendererClasses [&NiagaraMeshRendererProperties]}
 ]
 userVariables []
}
